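# Animation Blueprint for the SummerFlower backpack cosmetic, adding procedural secondary motion to its zipper and flower bones.
# Chained spring-driven, angle-limited AnimDynamics nodes simulate four zipper bones and five flower bones over the input pose.
# The chain then repeats: zippers switch from root-relative to Actor space with lower motion scales, while flowers stay base-relative with reduced angular influence.
# The EventGraph sets NotInFE from a front-end check, but no shown node reads it.
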AnimBlueprint /BRCosmetics/Accessories/FORT_Backpacks/Backpack_F_MED_SummerFlower/Meshes/F_MED_SummerFlower_Pack_AnimBP
target_skeleton=&/F_MED_SummerFlower_Pack_Skeleton
graphs [
 AnimGraph {nodes [
   n1 = Anim Dynamics - Bone: dyn_zipper_01!(ComponentPose=Local To Component(LocalPose=Input Pose()), bAlphaBoolEnabled:bool=True, ExternalForce="0,0,0") -> (Pose) {
    alphaInputType=Bool
    angularDampingOverride=0.995
    angularSpringConstant=1500
    bAngularSpring=true
    bGravityOverrideInSimSpace=true
    bOverrideAngularDamping=true
    bUseGravityOverride=true
    gravityScale=0.35
    simulationSpace=BoneRelative
    boundBone {boneName=dyn_zipper_01}
    componentAppliedLinearAccClamp=<3500 3500 3500>
    componentLinearAccScale=<0.6 0.6 0.6>
    componentLinearVelScale=<0.8 0.8 0.8>
    gravityOverride {z=-980}
    physicsBodyDefinitions [{
      boundBone {boneName=dyn_zipper_01}
      boxExtents=<10 15 15>
      constraintSetup {angularLimitsMax=<0 5 10>, angularLimitsMin=<0 -50 -10>, angularTarget=<1 0 0>}
      localJointOffset=<1 0 0>
     }]
    relativeSpaceBone {boneName=root}
   }
   n2 = Anim Dynamics - Bone: dyn_zipper_02!(ComponentPose=@n1, bAlphaBoolEnabled:bool=True, ExternalForce="0,0,0") -> (Pose) {
    alphaInputType=Bool
    angularDampingOverride=0.995
    angularSpringConstant=1500
    bAngularSpring=true
    bGravityOverrideInSimSpace=true
    bOverrideAngularDamping=true
    bUseGravityOverride=true
    gravityScale=0.35
    simulationSpace=BoneRelative
    boundBone {boneName=dyn_zipper_02}
    componentAppliedLinearAccClamp=<3500 3500 3500>
    componentLinearAccScale=<0.6 0.6 0.6>
    componentLinearVelScale=<0.8 0.8 0.8>
    gravityOverride {z=-980}
    physicsBodyDefinitions [{
      boundBone {boneName=dyn_zipper_02}
      boxExtents=<10 15 15>
      constraintSetup {angularLimitsMax=<0 2 10>, angularLimitsMin=<0 -30 -10>, angularTarget=<1 0 0>}
      localJointOffset=<1.5 0 0>
     }]
    relativeSpaceBone {boneName=root}
   }
   n3 = Anim Dynamics - Bone: dyn_zipper_03!(ComponentPose=@n2, bAlphaBoolEnabled:bool=True, ExternalForce="0,0,0") -> (Pose) {
    alphaInputType=Bool
    angularDampingOverride=0.995
    angularSpringConstant=1500
    bAngularSpring=true
    bGravityOverrideInSimSpace=true
    bOverrideAngularDamping=true
    bUseGravityOverride=true
    gravityScale=0.35
    simulationSpace=BoneRelative
    boundBone {boneName=dyn_zipper_03}
    componentAppliedLinearAccClamp=<3500 3500 3500>
    componentLinearAccScale=<0.6 0.6 0.6>
    componentLinearVelScale=<0.8 0.8 0.8>
    gravityOverride {z=-980}
    physicsBodyDefinitions [{
      boundBone {boneName=dyn_zipper_03}
      boxExtents=<10 15 15>
      constraintSetup {angularLimitsMax=<0 2 10>, angularLimitsMin=<0 -30 -10>, angularTarget=<1 0 0>}
      localJointOffset=<2 0 0>
     }]
    relativeSpaceBone {boneName=root}
   }
   n4 = Anim Dynamics - Bone: dyn_zipper_04!(ComponentPose=@n3, bAlphaBoolEnabled:bool=True, ExternalForce="0,0,0") -> (Pose) {
    alphaInputType=Bool
    angularDampingOverride=0.995
    angularSpringConstant=1500
    bAngularSpring=true
    bGravityOverrideInSimSpace=true
    bOverrideAngularDamping=true
    bUseGravityOverride=true
    gravityScale=0.35
    simulationSpace=BoneRelative
    boundBone {boneName=dyn_zipper_04}
    componentAppliedLinearAccClamp=<3500 3500 3500>
    componentLinearAccScale=<0.6 0.6 0.6>
    componentLinearVelScale=<0.8 0.8 0.8>
    gravityOverride {z=-980}
    physicsBodyDefinitions [{
      boundBone {boneName=dyn_zipper_04}
      boxExtents=<10 15 15>
      constraintSetup {angularLimitsMax=<0 5 10>, angularLimitsMin=<0 -30 -10>, angularTarget=<1 0 0>}
      localJointOffset=<3 0 0>
     }]
    relativeSpaceBone {boneName=root}
   }
   n5 = Anim Dynamics - Bone: dyn_flower_a!(ComponentPose=@n4, bAlphaBoolEnabled:bool=True, ExternalForce="0,0,0") -> (Pose) {
    alphaInputType=Bool
    angularDampingOverride=0.999
    angularSpringConstant=1200
    bAngularSpring=true
    bGravityOverrideInSimSpace=true
    bOverrideAngularDamping=true
    bUseGravityOverride=true
    gravityScale=0.35
    simulationSpace=BoneRelative
    boundBone {boneName=dyn_flower_a}
    gravityOverride {x=-400}
    physicsBodyDefinitions [{
      boundBone {boneName=dyn_flower_a}
      boxExtents=<15 10 10>
      constraintSetup {angularLimitsMax=<0 5 10>, angularLimitsMin=<0 -45 -10>, angularTarget=<1 0 0>}
      localJointOffset=<0 0 -0.5>
     }]
    relativeSpaceBone {boneName=dyn_flower_base}
    simSpaceSettings {maxAngularAcceleration=48, maxAngularVelocity=8, simSpaceAngularAlpha=1}
   }
   n6 = Anim Dynamics - Bone: dyn_flower_b!(ComponentPose=@n5, bAlphaBoolEnabled:bool=True, ExternalForce="0,0,0") -> (Pose) {
    alphaInputType=Bool
    angularDampingOverride=0.999
    angularSpringConstant=1200
    bAngularSpring=true
    bGravityOverrideInSimSpace=true
    bOverrideAngularDamping=true
    bUseGravityOverride=true
    gravityScale=0.35
    simulationSpace=BoneRelative
    boundBone {boneName=dyn_flower_b}
    gravityOverride {x=-400}
    physicsBodyDefinitions [{
      boundBone {boneName=dyn_flower_b}
      boxExtents=<15 10 10>
      constraintSetup {angularLimitsMax=<0 5 10>, angularLimitsMin=<0 -45 -10>, angularTarget=<1 0 0>}
      localJointOffset=<0 0 -0.5>
     }]
    relativeSpaceBone {boneName=dyn_flower_base}
    simSpaceSettings {maxAngularAcceleration=48, maxAngularVelocity=8, simSpaceAngularAlpha=1}
   }
   n7 = Anim Dynamics - Bone: dyn_flower_c!(ComponentPose=@n6, bAlphaBoolEnabled:bool=True, ExternalForce="0,0,0") -> (Pose) {
    alphaInputType=Bool
    angularDampingOverride=0.999
    angularSpringConstant=1200
    bAngularSpring=true
    bGravityOverrideInSimSpace=true
    bOverrideAngularDamping=true
    bUseGravityOverride=true
    gravityScale=0.35
    simulationSpace=BoneRelative
    boundBone {boneName=dyn_flower_c}
    gravityOverride {x=-400}
    physicsBodyDefinitions [{
      boundBone {boneName=dyn_flower_c}
      boxExtents=<15 10 10>
      constraintSetup {angularLimitsMax=<0 5 10>, angularLimitsMin=<0 -45 -10>, angularTarget=<1 0 0>}
      localJointOffset=<0 0 -0.5>
     }]
    relativeSpaceBone {boneName=dyn_flower_base}
    simSpaceSettings {maxAngularAcceleration=48, maxAngularVelocity=8, simSpaceAngularAlpha=1}
   }
   n8 = Anim Dynamics - Bone: dyn_flower_d!(ComponentPose=@n7, bAlphaBoolEnabled:bool=True, ExternalForce="0,0,0") -> (Pose) {
    alphaInputType=Bool
    angularDampingOverride=0.999
    angularSpringConstant=1200
    bAngularSpring=true
    bGravityOverrideInSimSpace=true
    bOverrideAngularDamping=true
    bUseGravityOverride=true
    gravityScale=0.35
    simulationSpace=BoneRelative
    boundBone {boneName=dyn_flower_d}
    gravityOverride {x=-400}
    physicsBodyDefinitions [{
      boundBone {boneName=dyn_flower_d}
      boxExtents=<15 10 10>
      constraintSetup {angularLimitsMax=<0 5 10>, angularLimitsMin=<0 -45 -10>, angularTarget=<1 0 0>}
      localJointOffset=<0 0 -0.5>
     }]
    relativeSpaceBone {boneName=dyn_flower_base}
    simSpaceSettings {maxAngularAcceleration=48, maxAngularVelocity=8, simSpaceAngularAlpha=1}
   }
   n9 = Anim Dynamics - Bone: dyn_flower_e!(ComponentPose=@n8, bAlphaBoolEnabled:bool=True, ExternalForce="0,0,0") -> (Pose) {
    alphaInputType=Bool
    angularDampingOverride=0.999
    angularSpringConstant=1200
    bAngularSpring=true
    bGravityOverrideInSimSpace=true
    bOverrideAngularDamping=true
    bUseGravityOverride=true
    gravityScale=0.35
    simulationSpace=BoneRelative
    boundBone {boneName=dyn_flower_e}
    gravityOverride {x=-400}
    physicsBodyDefinitions [{
      boundBone {boneName=dyn_flower_e}
      boxExtents=<15 10 10>
      constraintSetup {angularLimitsMax=<0 5 10>, angularLimitsMin=<0 -45 -10>, angularTarget=<1 0 0>}
      localJointOffset=<0 0 -0.5>
     }]
    relativeSpaceBone {boneName=dyn_flower_base}
    simSpaceSettings {maxAngularAcceleration=48, maxAngularVelocity=8, simSpaceAngularAlpha=1}
   }
   n10 = Anim Dynamics - Bone: dyn_zipper_01(ComponentPose=@n9, bAlphaBoolEnabled=True, ExternalForce="0,0,0") {
    alphaInputType=Bool
    angularDampingOverride=0.995
    angularSpringConstant=1500
    bAngularSpring=true
    bGravityOverrideInSimSpace=true
    bOverrideAngularDamping=true
    bUseGravityOverride=true
    gravityScale=0.35
    simulationSpace=Actor
    boundBone {boneName=dyn_zipper_01}
    componentAppliedLinearAccClamp=<3000 3000 3000>
    componentLinearAccScale=<0.4 0.4 0.4>
    componentLinearVelScale=<0.6 0.6 0.6>
    gravityOverride {z=-980}
    physicsBodyDefinitions [{
      boundBone {boneName=dyn_zipper_01}
      boxExtents=<10 15 15>
      constraintSetup {angularLimitsMax=<0 5 10>, angularLimitsMin=<0 -50 -10>, angularTarget=<1 0 0>}
      localJointOffset=<1 0 0>
     }]
    relativeSpaceBone {boneName=root}
   }
   n11 = Anim Dynamics - Bone: dyn_zipper_02(ComponentPose=@n10, bAlphaBoolEnabled=True, ExternalForce="0,0,0") {
    alphaInputType=Bool
    angularDampingOverride=0.995
    angularSpringConstant=1500
    bAngularSpring=true
    bGravityOverrideInSimSpace=true
    bOverrideAngularDamping=true
    bUseGravityOverride=true
    gravityScale=0.35
    simulationSpace=Actor
    boundBone {boneName=dyn_zipper_02}
    componentAppliedLinearAccClamp=<3000 3000 3000>
    componentLinearAccScale=<0.4 0.4 0.4>
    componentLinearVelScale=<0.6 0.6 0.6>
    gravityOverride {z=-980}
    physicsBodyDefinitions [{
      boundBone {boneName=dyn_zipper_02}
      boxExtents=<10 15 15>
      constraintSetup {angularLimitsMax=<0 2 10>, angularLimitsMin=<0 -30 -10>, angularTarget=<1 0 0>}
      localJointOffset=<1.5 0 0>
     }]
    relativeSpaceBone {boneName=root}
   }
   n12 = Anim Dynamics - Bone: dyn_zipper_03(ComponentPose=@n11, bAlphaBoolEnabled=True, ExternalForce="0,0,0") {
    alphaInputType=Bool
    angularDampingOverride=0.995
    angularSpringConstant=1500
    bAngularSpring=true
    bGravityOverrideInSimSpace=true
    bOverrideAngularDamping=true
    bUseGravityOverride=true
    gravityScale=0.35
    simulationSpace=Actor
    boundBone {boneName=dyn_zipper_03}
    componentAppliedLinearAccClamp=<3000 3000 3000>
    componentLinearAccScale=<0.4 0.4 0.4>
    componentLinearVelScale=<0.6 0.6 0.6>
    gravityOverride {z=-980}
    physicsBodyDefinitions [{
      boundBone {boneName=dyn_zipper_03}
      boxExtents=<10 15 15>
      constraintSetup {angularLimitsMax=<0 2 10>, angularLimitsMin=<0 -30 -10>, angularTarget=<1 0 0>}
      localJointOffset=<2 0 0>
     }]
    relativeSpaceBone {boneName=root}
   }
   n13 = Anim Dynamics - Bone: dyn_zipper_04(ComponentPose=@n12, bAlphaBoolEnabled=True, ExternalForce="0,0,0") {
    alphaInputType=Bool
    angularDampingOverride=0.995
    angularSpringConstant=1500
    bAngularSpring=true
    bGravityOverrideInSimSpace=true
    bOverrideAngularDamping=true
    bUseGravityOverride=true
    gravityScale=0.35
    simulationSpace=Actor
    boundBone {boneName=dyn_zipper_04}
    componentAppliedLinearAccClamp=<3000 3000 3000>
    componentLinearAccScale=<0.4 0.4 0.4>
    componentLinearVelScale=<0.6 0.6 0.6>
    gravityOverride {z=-980}
    physicsBodyDefinitions [{
      boundBone {boneName=dyn_zipper_04}
      boxExtents=<10 15 15>
      constraintSetup {angularLimitsMax=<0 5 10>, angularLimitsMin=<0 -30 -10>, angularTarget=<1 0 0>}
      localJointOffset=<3 0 0>
     }]
    relativeSpaceBone {boneName=root}
   }
   n14 = Anim Dynamics - Bone: dyn_flower_a(ComponentPose=@n13, bAlphaBoolEnabled=True, ExternalForce="0,0,0") {
    alphaInputType=Bool
    angularDampingOverride=0.999
    angularSpringConstant=1200
    bAngularSpring=true
    bGravityOverrideInSimSpace=true
    bOverrideAngularDamping=true
    bUseGravityOverride=true
    gravityScale=0.35
    simulationSpace=BoneRelative
    boundBone {boneName=dyn_flower_a}
    gravityOverride {x=-400}
    physicsBodyDefinitions [{
      boundBone {boneName=dyn_flower_a}
      boxExtents=<15 10 10>
      constraintSetup {angularLimitsMax=<0 5 10>, angularLimitsMin=<0 -45 -10>, angularTarget=<1 0 0>}
      localJointOffset=<0 0 -0.5>
     }]
    relativeSpaceBone {boneName=dyn_flower_base}
    simSpaceSettings {maxAngularAcceleration=24, maxAngularVelocity=4, simSpaceAngularAlpha=0.4}
   }
   n15 = Anim Dynamics - Bone: dyn_flower_b(ComponentPose=@n14, bAlphaBoolEnabled=True, ExternalForce="0,0,0") {
    alphaInputType=Bool
    angularDampingOverride=0.999
    angularSpringConstant=1200
    bAngularSpring=true
    bGravityOverrideInSimSpace=true
    bOverrideAngularDamping=true
    bUseGravityOverride=true
    gravityScale=0.35
    simulationSpace=BoneRelative
    boundBone {boneName=dyn_flower_b}
    gravityOverride {x=-400}
    physicsBodyDefinitions [{
      boundBone {boneName=dyn_flower_b}
      boxExtents=<15 10 10>
      constraintSetup {angularLimitsMax=<0 5 10>, angularLimitsMin=<0 -45 -10>, angularTarget=<1 0 0>}
      localJointOffset=<0 0 -0.5>
     }]
    relativeSpaceBone {boneName=dyn_flower_base}
    simSpaceSettings {maxAngularAcceleration=24, maxAngularVelocity=4, simSpaceAngularAlpha=0.4}
   }
   n16 = Anim Dynamics - Bone: dyn_flower_c(ComponentPose=@n15, bAlphaBoolEnabled=True, ExternalForce="0,0,0") {
    alphaInputType=Bool
    angularDampingOverride=0.999
    angularSpringConstant=1200
    bAngularSpring=true
    bGravityOverrideInSimSpace=true
    bOverrideAngularDamping=true
    bUseGravityOverride=true
    gravityScale=0.35
    simulationSpace=BoneRelative
    boundBone {boneName=dyn_flower_c}
    gravityOverride {x=-400}
    physicsBodyDefinitions [{
      boundBone {boneName=dyn_flower_c}
      boxExtents=<15 10 10>
      constraintSetup {angularLimitsMax=<0 5 10>, angularLimitsMin=<0 -45 -10>, angularTarget=<1 0 0>}
      localJointOffset=<0 0 -0.5>
     }]
    relativeSpaceBone {boneName=dyn_flower_base}
    simSpaceSettings {maxAngularAcceleration=24, maxAngularVelocity=4, simSpaceAngularAlpha=0.4}
   }
   n17 = Anim Dynamics - Bone: dyn_flower_d(ComponentPose=@n16, bAlphaBoolEnabled=True, ExternalForce="0,0,0") {
    alphaInputType=Bool
    angularDampingOverride=0.999
    angularSpringConstant=1200
    bAngularSpring=true
    bGravityOverrideInSimSpace=true
    bOverrideAngularDamping=true
    bUseGravityOverride=true
    gravityScale=0.35
    simulationSpace=BoneRelative
    boundBone {boneName=dyn_flower_d}
    gravityOverride {x=-400}
    physicsBodyDefinitions [{
      boundBone {boneName=dyn_flower_d}
      boxExtents=<15 10 10>
      constraintSetup {angularLimitsMax=<0 5 10>, angularLimitsMin=<0 -45 -10>, angularTarget=<1 0 0>}
      localJointOffset=<0 0 -0.5>
     }]
    relativeSpaceBone {boneName=dyn_flower_base}
    simSpaceSettings {maxAngularAcceleration=24, maxAngularVelocity=4, simSpaceAngularAlpha=0.4}
   }
   n18 = Anim Dynamics - Bone: dyn_flower_e(ComponentPose=@n17, bAlphaBoolEnabled=True, ExternalForce="0,0,0") {
    alphaInputType=Bool
    angularDampingOverride=0.999
    angularSpringConstant=1200
    bAngularSpring=true
    bGravityOverrideInSimSpace=true
    bOverrideAngularDamping=true
    bUseGravityOverride=true
    gravityScale=0.35
    simulationSpace=BoneRelative
    boundBone {boneName=dyn_flower_e}
    gravityOverride {x=-400}
    physicsBodyDefinitions [{
      boundBone {boneName=dyn_flower_e}
      boxExtents=<15 10 10>
      constraintSetup {angularLimitsMax=<0 5 10>, angularLimitsMin=<0 -45 -10>, angularTarget=<1 0 0>}
      localJointOffset=<0 0 -0.5>
     }]
    relativeSpaceBone {boneName=dyn_flower_base}
    simSpaceSettings {maxAngularAcceleration=24, maxAngularVelocity=4, simSpaceAngularAlpha=0.4}
   }
   Output Pose(Result=Component To Local(ComponentPose=@n18)) {layerGroup=None}
  ]}
 EventGraph {nodes [
   Event BlueprintUpdateAnimation()
   > Set NotInFE(NotInFE:bool=NOT Boolean(A=Get bIsInFrontEnd())) -> (Output_Get:bool)
   TryGetPawnOwner()
  ]}
]
variables [NotInFE {type=boolean}]
